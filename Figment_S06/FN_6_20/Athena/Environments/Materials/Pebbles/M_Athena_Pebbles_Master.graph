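# Master material for instanced pebble meshes in the Athena environment, using the Default Lit shading model.
# The diffuse texture is tinted by Diffuse_Tint or Arid_Tint, blended by the Swamp output of the landscape biome mask function, despite the arid parameter name.
# The diffuse red channel also blends RMin and RMax to scale a Roughness parameter, alongside a Specular parameter and a normal map.
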Material /Figment_S06/FN_6_20/Athena/Environments/Materials/Pebbles/M_Athena_Pebbles_Master
graph {
 properties {bUsedWithInstancedStaticMeshes=true, pixelDepthOffsetMode=PDOM_Legacy, usedShadingModels=Default Lit}
 nodes [
  n1 = TextureSample() {texture=/Figment_S06/FN_6_20/Athena/Environments/Textures/Pebbles/T_Athena_Pebbles_D}
  LinearInterpolate(
   A=Multiply(A=@n1.RGB, B=VectorParameter() {parameterName=Diffuse_Tint, defaultValue=<rgba 1 0.8183 0.7059 0>}.RGB)
   B=Multiply(A=@n1.RGB, B=VectorParameter() {parameterName=Arid_Tint, defaultValue=<rgba 2 0.7697 0.3058 1>}.RGB)
   Alpha=MaterialFunctionCall() {materialFunction=/FigmentEnvironment/FN_6_20/Athena/Environments/Landscape/MaterialFunctions/MF_Athena_BiomeMask}.Swamp
  )
  Multiply(
   A=ScalarParameter() {defaultValue=0.6095, parameterName=Roughness}
   B=LinearInterpolate(A=ScalarParameter() {defaultValue=1, parameterName=RMin}, B=ScalarParameter() {defaultValue=1, parameterName=RMax}, Alpha=@n1.R)
  )
  ScalarParameter() {defaultValue=0.5, parameterName=Specular}
  TextureSample() {samplerType=SAMPLERTYPE_Normal, texture=/Figment_S06/FN_6_20/Athena/Environments/Textures/Pebbles/T_Athena_Pebbles_N}
 ]
}
